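# Material function for the BlightShield01 wheel that layers view-angle color tints and added specular onto the output of MF_WheelParent_Simple.
# The rim tint uses a biased, scaled dot product of a transformed Z vector and the camera vector, masked by the parent's RimRGBAMask alpha.
# The tire tint combines plain and normal-mapped Fresnel terms.
# Vertex color red blends rim and tire results for base color, emissive and texture-masked specular.
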MaterialFunction /VehicleCosmetics/Wheels/BlightShield01/Materials/MF_Wheel_BlightShield01_Simple
graph {nodes [
  n1 = NamedRerouteUsage()
  n2 = VectorParameter() {group=2_Rim Properties, parameterName=Rim_FresnelColor, sortPriority=0, defaultValue=<rgba 0.426 0.0199 0.0254 1>}
  n3 = DotProduct(A=Transform(None=Constant3Vector() {constant=<rgba 0 0 1 0>}), B=CameraVectorWS())
  n4 = Saturate(None=ConstantBiasScale(None=OneMinus(None=@n3)) {bias=-0.5, scale=5})
  n5 = LinearInterpolate(A=@n1, B=@n2.RGB, Alpha=@n4)
  n6 = ComponentMask(None=NamedRerouteUsage()) {a=true, desc=Use RGBA Alpha to mask fresnel}
  n7 = NamedRerouteUsage()
  n8 = VectorParameter() {group=3_Tire Properties, parameterName=Tire_FresnelColor, defaultValue=<rgba 0.1179 0.0193 0.0207 1>}
  n9 = Fresnel() {baseReflectFraction=0.1}
  n10 = Fresnel(Normal=Transform(None=NamedRerouteUsage())) {exponent=0.5}
  n11 = OneMinus(None=Multiply(A=ConstantBiasScale(None=@n9) {bias=0.03, scale=22}, B=OneMinus(None=@n10)))
  n12 = NamedRerouteUsage()
  n13 = LinearInterpolate(A=@n12, B=@n2.RGB, Alpha=@n4) {desc=Rim Emissive}
  n14 = ComponentMask(None=NamedRerouteUsage()) {a=true, desc=Use RGBA Alpha to mask fresnel}
  n15 = LinearInterpolate(A=@n12, B=@n8.RGB, Alpha=Saturate(None=@n11)) {desc=Tire}
  n16 = VectorParameter() {group=2_Rim Properties, parameterName=Rim_AddedSpecularColor, sortPriority=0, defaultValue=<rgba 1 0.583 0.358 1>}
  n17 = ComponentMask(None=NamedRerouteUsage()) {r=true}
  n18 = VectorParameter() {group=3_Tire Properties, parameterName=Tire_AddedSpecularColor, sortPriority=0, defaultValue=<rgba 0.426 0.0199 0.0254 1>}
  n19 = TextureSampleParameter2D() {
   group=1_Input Textures
   parameterName=Tire_RGB_Special
   samplerType=SAMPLERTYPE_LinearColor
   texture=/VehicleCosmetics/Wheels/BlightShield01/Textures/T_Tire_BlightShield01_RGB
  }
  n20 = SetMaterialAttributes!(
   MaterialAttributes=NamedRerouteUsage()
   BaseColor=LinearInterpolate(A=LinearInterpolate(A=@n1, B=@n5, Alpha=@n6), B=LinearInterpolate(A=@n7, B=@n8.RGB, Alpha=@n11), Alpha=VertexColor().R) {constB=0, desc=BC}
   EmissiveColor=LinearInterpolate(A=LinearInterpolate(B=@n13, Alpha=@n14), B=Multiply(A=@n15) {constB=1.5, desc=Tire}, Alpha=VertexColor().R)
   Metallic
   Specular=LinearInterpolate(A=Multiply(A=@n16.RGB, B=@n17), B=Multiply(A=@n18.RGB, B=@n19.R), Alpha=VertexColor().R) {constB=0, desc=Specular}
  ) {attributeSetTypes [
    #1
    #2
    #3
    #4
   ]}
  FunctionOutput(None=@n20)
  n21 = MaterialFunctionCall() {materialFunction=&MF_WheelParent_Simple}
  n22 = GetMaterialAttributes(None=@n21.MaterialAttributes) {attributeGetTypes [#1, #5, #2]}
  NamedRerouteDeclaration(None=@n22.BaseColor) {name=BC_Out, nodeColor=<rgba 0 0.1616 0.5226 1>}
  NamedRerouteDeclaration(None=@n22.Normal) {name=Normal_Out, nodeColor=<rgba 0.1902 0.9979 1 1>}
  NamedRerouteDeclaration(None=@n21.RimRGBAMask) {name=RimRGB_Out, nodeColor=<rgba 1 0 0.7537 1>}
  NamedRerouteDeclaration(None=@n21.MaterialAttributes) {name=MatAttributes_Out, nodeColor=<rgba 1 0 0.7969 1>}
  NamedRerouteDeclaration(None=@n22.EmissiveColor) {name=Emissive_Out, nodeColor=<rgba 0 1 0.8828 1>}
 ]}
